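# Physics asset for the SirWolf tail accessory (M_MED_SirWolf), defining ragdoll-style simulation bodies along its dynamic tail bones.
# Seven identical short, fat capsules cover the tail joints, hanging from a zero-size body on the tail base joint.
# Constraints link each joint to the one before it in a single chain, with matching limited swing on both axes and twist limit set to zero.
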PhysicsAsset /Game/Accessories/FORT_Tails/Tail_M_MED_SirWolf/Meshes/M_MED_SirWolf_Tail_Physics1
bodies {
 C_Tail_dyn_A_1_Jnt [{shapeType=Capsule, length=3, radius=5, shapeName=C_Tail_dyn_A_1_Jnt_capsule, center=<5 0 0>, rotation=<rot -90 0 0>}]
 C_Tail_dyn_A_2_Jnt [{shapeType=Capsule, length=3, radius=5, shapeName=C_Tail_dyn_A_2_Jnt_capsule, center=<5 0 0>, rotation=<rot -90 0 0>}]
 C_Tail_dyn_A_3_Jnt [{shapeType=Capsule, length=3, radius=5, shapeName=C_Tail_dyn_A_3_Jnt_capsule, center=<5 0 0>, rotation=<rot -90 0 0>}]
 C_Tail_dyn_A_4_Jnt [{shapeType=Capsule, length=3, radius=5, shapeName=C_Tail_dyn_A_4_Jnt_capsule, center=<5 0 0>, rotation=<rot -90 0 0>}]
 C_Tail_dyn_A_5_Jnt [{shapeType=Capsule, length=3, radius=5, shapeName=C_Tail_dyn_A_5_Jnt_capsule, center=<5 0 0>, rotation=<rot -90 0 0>}]
 C_Tail_dyn_A_6_Jnt [{shapeType=Capsule, length=3, radius=5, shapeName=C_Tail_dyn_A_6_Jnt_capsule, center=<5 0 0>, rotation=<rot -90 0 0>}]
 C_Tail_dyn_A_7_Jnt [{shapeType=Capsule, length=3, radius=5, shapeName=C_Tail_dyn_A_7_Jnt_capsule, center=<5 0 0>, rotation=<rot -90 0 0>}]
 C_Tail_dyn_A_Base_Jnt2 [{shapeType=Capsule, length=0, radius=0, shapeName=C_Tail_dyn_A_Base_Jnt2_capsule}]
}
constraints [
 {bone1Name=C_Tail_dyn_A_1_Jnt, bone2Name=C_Tail_dyn_A_Base_Jnt2, swing1LimitDegrees=45, swing2LimitDegrees=45, twistLimitDegrees=0}
 {bone1Name=C_Tail_dyn_A_2_Jnt, bone2Name=C_Tail_dyn_A_1_Jnt, swing1LimitDegrees=45, swing2LimitDegrees=45, twistLimitDegrees=0}
 {bone1Name=C_Tail_dyn_A_3_Jnt, bone2Name=C_Tail_dyn_A_2_Jnt, swing1LimitDegrees=45, swing2LimitDegrees=45, twistLimitDegrees=0}
 {bone1Name=C_Tail_dyn_A_4_Jnt, bone2Name=C_Tail_dyn_A_3_Jnt, swing1LimitDegrees=45, swing2LimitDegrees=45, twistLimitDegrees=0}
 {bone1Name=C_Tail_dyn_A_5_Jnt, bone2Name=C_Tail_dyn_A_4_Jnt, swing1LimitDegrees=45, swing2LimitDegrees=45, twistLimitDegrees=0}
 {bone1Name=C_Tail_dyn_A_6_Jnt, bone2Name=C_Tail_dyn_A_5_Jnt, swing1LimitDegrees=45, swing2LimitDegrees=45, twistLimitDegrees=0}
 {bone1Name=C_Tail_dyn_A_7_Jnt, bone2Name=C_Tail_dyn_A_6_Jnt, swing1LimitDegrees=45, swing2LimitDegrees=45, twistLimitDegrees=0}
]
